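# Animation Blueprint for the Chrono Pan Rifle that spins its circular magazine and plays an overshoot motion when firing stops.
# A base pose passes through DefaultSlot, gets an additive circle_magazine yaw from MagRotationValue, and is cached as PreAdditive.
# A state machine layers the additive overshoot over that cache when firing ends without reloading, returning on completion or instantly on refiring.
# The event graph is empty and the driving variables are not declared here.
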AnimBlueprint /ChronoWeaponGameplay/Items/PanRifle/Animations/ChronoPan_Rifle_AnimBP
target_skeleton=/ChronoWeaponGameplay/Items/PanRifle/Meshes/Chrono_Pan_Rifle_Skeleton
graphs [
 AnimGraph {nodes [
   AnimGraphNode_StateMachine_0 = MagOvershoot_StateMachine() {is=title AnimGraphNode_StateMachine}
   Output Pose(Result=@AnimGraphNode_StateMachine_0) {layerGroup=None}
   n1 = Sequence Player 'Chrono_Pan_MagazineOvershoot_BasePose_Chrono_Pan_Rifle'() {bLoopAnimation=false, sequence=/ChronoWeaponGameplay/Items/PanRifle/Animations/Proto/Chrono_Pan_MagazineOvershoot_BasePose_Chrono_Pan_Rifle}
   n2 = "Transform (Modify) Bone - Bone: circle_magazine"!(
    ComponentPose=Local To Component(LocalPose=Slot 'DefaultSlot'(Source=@n1))
    Alpha=1
    Rotation=MakeRotator(Yaw=Get MagRotationValue() -> (MagRotationValue:float))
   ) -> (Pose) {rotationMode=BMM_Additive, rotationSpace=BCS_BoneSpace, alphaBoolBlend {blendOption=Linear}, boneToModify {boneName=circle_magazine}}
   Save cached pose 'PreAdditive'(Pose=Component To Local(ComponentPose=@n2)) {cacheName=PreAdditive}
  ]}
 MagOvershoot_StateMachine {
  entry=CurrentPose
  path=AnimGraph.AnimGraphNode_StateMachine_0.MagOvershoot_StateMachine
  ids=id
  nodes [
   AnimStateNode_0 = CurrentPose() {is=title AnimStateNode}
   AnimStateNode_1 = Shooting() {is=title AnimStateNode}
   AnimStateNode_2 = Overshoot() {is=title AnimStateNode}
   AnimStateTransitionNode_0 = AnimStateTransitionNode()
   AnimStateTransitionNode_1 = AnimStateTransitionNode()
   AnimStateTransitionNode_2 = AnimStateTransitionNode() {bAutomaticRuleBasedOnSequencePlayerInState=true}
   AnimStateTransitionNode_3 = AnimStateTransitionNode() {crossfadeDuration=0}
  ]
  links [
   {
    dst=Shooting
    rule=AnimGraph.AnimGraphNode_StateMachine_0.MagOvershoot_StateMachine.AnimStateTransitionNode_0.Transition
    src=CurrentPose
    via=AnimStateTransitionNode_0
   }
   {
    dst=CurrentPose
    rule=AnimGraph.AnimGraphNode_StateMachine_0.MagOvershoot_StateMachine.AnimStateTransitionNode_2.Transition
    src=Overshoot
    via=AnimStateTransitionNode_2
   }
   {
    dst=CurrentPose
    rule=AnimGraph.AnimGraphNode_StateMachine_0.MagOvershoot_StateMachine.AnimStateTransitionNode_3.Transition
    src=Overshoot
    via=AnimStateTransitionNode_3
   }
   {
    dst=Overshoot
    rule=AnimGraph.AnimGraphNode_StateMachine_0.MagOvershoot_StateMachine.AnimStateTransitionNode_1.Transition
    src=Shooting
    via=AnimStateTransitionNode_1
   }
  ]
 }
 CurrentPose {
  path=AnimGraph.AnimGraphNode_StateMachine_0.MagOvershoot_StateMachine.AnimStateNode_0.CurrentPose
  nodes [Output Animation Pose(Result=Use cached pose 'PreAdditive'())]
 }
 Shooting {
  path=AnimGraph.AnimGraphNode_StateMachine_0.MagOvershoot_StateMachine.AnimStateNode_1.Shooting
  nodes [Output Animation Pose(Result=Use cached pose 'PreAdditive'())]
 }
 Overshoot {
  path=AnimGraph.AnimGraphNode_StateMachine_0.MagOvershoot_StateMachine.AnimStateNode_2.Overshoot
  nodes [Output Animation Pose(Result=Apply Additive(
     Base=Use cached pose 'PreAdditive'()
     Additive="Sequence Player (additive) 'Chrono_Pan_MagazineOvershoot_Chrono_Pan_Rifle'"() {bLoopAnimation=false, sequence=/ChronoWeaponGameplay/Items/PanRifle/Animations/Proto/Chrono_Pan_MagazineOvershoot_Chrono_Pan_Rifle}
    ) {alphaBoolBlend {blendOption=Linear}})]
 }
 Transition {
  path=AnimGraph.AnimGraphNode_StateMachine_0.MagOvershoot_StateMachine.AnimStateTransitionNode_0.Transition
  nodes [Result(bCanEnterTransition=Get bIsFiring() -> (bIsFiring:bool))]
 }
 Transition {
  path=AnimGraph.AnimGraphNode_StateMachine_0.MagOvershoot_StateMachine.AnimStateTransitionNode_1.Transition
  nodes [
   n1 = Get bIsReloading() -> (bIsReloading:bool)
   n2 = Get bIsFiring()
   Result(bCanEnterTransition=AND Boolean(A=NOT Boolean(A=@n2), B=NOT Boolean(A=@n1)))
  ]
 }
 Transition {path=AnimGraph.AnimGraphNode_StateMachine_0.MagOvershoot_StateMachine.AnimStateTransitionNode_2.Transition, nodes [Result()]}
 Transition {
  path=AnimGraph.AnimGraphNode_StateMachine_0.MagOvershoot_StateMachine.AnimStateTransitionNode_3.Transition
  nodes [Result(bCanEnterTransition=Get bIsFiring())]
 }
 EventGraph {links [], nodes []}
]
variables []
