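# ChronoPan_AR_ADS_In_Cue is the sound cue for aiming down sights on the Pan rifle in the Chrono weapon gameplay content.
# A single wave player feeds the ADS-in wave straight to the output, with no randomization or modulation.
# It plays at reduced volume in the weapon foley sound class, borrowing the shared weapon reload attenuation for spatial falloff.
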SoundCue /ChronoWeaponGameplay/Items/PanRifle/Audio/ChronoPan_AR_ADS_In_Cue
root=WavePlayer_1
nodes [{id=WavePlayer_1, properties {soundWaveAssetPtr=&/ChronoPan_AR_ADS_In_01}}]
properties {
 attenuationSettings=&Fort_Weapon_Reload_Att
 maxDistance=2250
 soundClassObject=&Fort_WeaponFoley
 volumeMultiplier=0.301
}
