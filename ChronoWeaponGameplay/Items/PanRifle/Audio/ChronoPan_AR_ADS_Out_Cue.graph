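# Sound cue for the ChronoPan assault rifle's ADS-out foley, presumably heard when the weapon leaves aim-down-sights.
# A single wave player feeds one sound wave straight to the output, with no randomization or modulation.
# It uses the weapon foley sound class and borrows the weapon reload attenuation, with reduced volume.
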SoundCue /ChronoWeaponGameplay/Items/PanRifle/Audio/ChronoPan_AR_ADS_Out_Cue
root=WavePlayer_0
nodes [{id=WavePlayer_0, properties {soundWaveAssetPtr=&/ChronoPan_AR_ADS_Out_01}}]
properties {
 attenuationSettings=&Fort_Weapon_Reload_Att
 maxDistance=2250
 soundClassObject=&Fort_WeaponFoley
 volumeMultiplier=0.301
}
