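# Sound cue for the ChronoWeapon Pan Rifle reload foley, named for the magazine coming out of a pouch.
# A Random node picks between two wave variations with equal weight.
# It uses the shared weapon reload attenuation and weapon foley sound class at about half volume.
# It overrides concurrency and has the debug flag switched on.
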SoundCue /ChronoWeaponGameplay/Items/PanRifle/Audio/ChronoPan_AR_Reload_Mag_Out_Pouch_Cue
root=Random_1
nodes [
 {id=Random_1, children [WavePlayer_1,WavePlayer_3], properties {weights [1,1]}}
 {id=WavePlayer_1, properties {soundWaveAssetPtr=&/ChronoPan_AR_Reload_Mag_Out_Pouch_01}}
 {id=WavePlayer_3, properties {soundWaveAssetPtr=&/ChronoPan_AR_Reload_Mag_Out_Pouch_02}}
]
properties {
 attenuationSettings=&Fort_Weapon_Reload_Att
 bDebug=true
 bOverrideConcurrency=true
 maxDistance=2250
 soundClassObject=&Fort_WeaponFoley
 volumeMultiplier=0.501
}
