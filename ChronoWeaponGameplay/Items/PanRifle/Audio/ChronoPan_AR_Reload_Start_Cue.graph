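# Sound cue for the start of the ChronoPan assault rifle reload in the PanRifle weapon's audio set.
# A random node picks between two equally weighted wave variations.
# It routes through the weapon foley sound class with shared reload attenuation, a reduced volume and a concurrency override, and has its debug flag left on.
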SoundCue /ChronoWeaponGameplay/Items/PanRifle/Audio/ChronoPan_AR_Reload_Start_Cue
root=Random_2
nodes [
 {id=Random_2, children [WavePlayer_5,WavePlayer_6], properties {weights [1,1]}}
 {id=WavePlayer_5, properties {soundWaveAssetPtr=&/ChronoPan_AR_Reload_Start_01}}
 {id=WavePlayer_6, properties {soundWaveAssetPtr=&/ChronoPan_AR_Reload_Start_02}}
]
properties {
 attenuationSettings=&Fort_Weapon_Reload_Att
 bDebug=true
 bOverrideConcurrency=true
 maxDistance=2250
 soundClassObject=&Fort_WeaponFoley
 volumeMultiplier=0.501
}
