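; Weapon Blueprint for the Chrono Pan Rifle assault rifle, adding cosmetic impact and shell-ejection effects on top of its parent weapon class.
; On impact it calls the parent, then spawns a radial-force physics field at the hit location, only at one visual effects quality tier.
; ActivateShellsFX passes burst and looping Niagara shell systems, tuned by shell variables, to the native activation function.
; Fire FX, targeting and construction only call the parent; BeginPlay and Tick are empty.
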(event EventBeginPlay)

(event EventTick
  (params (DeltaSeconds :type Float%20%28single-precision%29)))

(event EventOnPlayWeaponFireFX
  (params (bPersistentFire :type Boolean) (bSecondaryFire :type Boolean))
  (|CallParentFunction bPersistentFire bSecondaryFire))

(event EventOnSetTargeting
  (params (bNewIsTargeting :type Boolean))
  (|CallParentFunction bNewIsTargeting))

(event FortWeapon|EventOnPlayImpactFX
  (params (HitResult :type Hit%20Result%20Structure%20%28by%20ref%29) (ImpactPhysicalSurface :type EPhysicalSurface%20Enum :default "SurfaceType_Default") (SpawnedPSC :type FXSystem%20Component%20Object%20Reference))
  (bind (_bblockinghit _binitialoverlap _time _distance _location) (Collision|BreakHitResult HitResult))
  (|CallParentFunction HitResult ImpactPhysicalSurface SpawnedPSC)
  (bind _returnvalue (Settings|GetGameUserSettings))
  (switch Utilities|FlowControl|Switch|SwitchonBPScalabilitySelection (Utilities|Enum|BytetoEnumBPScalabilitySelection (Math|Conversions|ToByte(Integer) (Settings|GetVisualEffectQuality _returnvalue)))
    (:NewEnumerator3
      (Game|SpawnActorfromClass "/Game/Effects/Niagara/Blueprints/PhsyicsFields/FN_RadialForce.FN_RadialForce_C" (Math|Transform|MakeTransform _location) "Undefined" "MultiplyWithRoot" 0 0.2 0.2 185.0 1.25 0.0 0))))

(fn ActivateShellsFX
  (params (Bool :type Boolean))
  (if Bool
    (CallFunction|ActivateShellsFXNative "/ChronoWeaponGameplay/Items/PanRifle/FX/NS_PanRifle_ShellsFire_Burst.NS_PanRifle_ShellsFire_Burst" "/ChronoWeaponGameplay/Items/PanRifle/FX/NS_PanRifle_ShellsFire_Looping.NS_PanRifle_ShellsFire_Looping" (Variables|Effects|Shells|GetShells%20Socket%20Name) "ShellsRotationRate" (Variables|Effects|Shells|GetShellsRotationRate) "Velocity" (Variables|Effects|Shells|GetShells%20Velocity) "Gravity" (Variables|Effects|Shells|GetShells%20Gravity) "Size" (Variables|Effects|Shells|GetShells%20Size) "InheritParentVelocity" (Math|Conversions|ToVector(Float) (Variables|Effects|Shells|GetInherit%20Parent%20Velocity)) "CylinderRadius" (Variables|Effects|Shells|GetCylindrical%20Radius) "CylinderHeight" (Variables|Effects|Shells|GetCylindrical%20Height) "ShellColor" (Variables|Effects|Shells|GetShell%20Color) "Shell Type" (Utilities|Select :Index (Variables|Effects|Shells|GetShellTypeSelect) :NewEnumerator0 0 :NewEnumerator1 1 :NewEnumerator2 2 :NewEnumerator3 4 :NewEnumerator4 3) (Variables|Effects|DrawDistances|GetShells%20PS%20Max%20Draw%20Distance) (Variables|Effects|Shells|GetUseShellsOnReload?) "SpawnProbability" (Variables|Effects|Shells|GetShells%20Spawn%20Rate%20Scale))))

(fn ConstructionScript
  (|CallParentFunction))
