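; Anim notify state that toggles the barrel glow effect on the Chrono Pan assault rifle during the animation window it covers.
; On begin and end it casts the mesh owner to B_Assault_Chrono_Pan_Rifle2_Athena and checks that it is valid.
; It then sets the "Fire" bool on that weapon's barrel glow Niagara component, true on begin and false on end.
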(fn Received_NotifyBegin
  (params (MeshComp :type Skeletal%20Mesh%20Component%20Object%20Reference) (Animation :type Anim%20Sequence%20Base%20Object%20Reference) (TotalDuration :type Float%20%28single-precision%29) (EventReference :type Anim%20Notify%20Event%20Reference%20Structure%20%28by%20ref%29))
  (bind _asb_assault_chrono_pan_rifle_2_athena (Utilities|Casting|CastToB_Assault_Chrono_Pan_Rifle2_Athena (Components|GetOwner MeshComp)))
  (Utilities|IsValid _asb_assault_chrono_pan_rifle_2_athena
    (:Is%20Valid
      (Niagara|SetNiagaraVariable(Bool) (Variables|/ChronoWeaponGameplay/Items/PanRifle/B_Assault_Chrono_Pan_Rifle2_Athena.B_Assault_Chrono_Pan_Rifle2_Athena_C|GetNS_FX_Barrel_Glow _asb_assault_chrono_pan_rifle_2_athena) "Fire" true)
      (return false))))

(fn Received_NotifyEnd
  (params (MeshComp :type Skeletal%20Mesh%20Component%20Object%20Reference) (Animation :type Anim%20Sequence%20Base%20Object%20Reference) (EventReference :type Anim%20Notify%20Event%20Reference%20Structure%20%28by%20ref%29))
  (bind _asb_assault_chrono_pan_rifle_2_athena (Utilities|Casting|CastToB_Assault_Chrono_Pan_Rifle2_Athena (Components|GetOwner MeshComp)))
  (Utilities|IsValid _asb_assault_chrono_pan_rifle_2_athena
    (:Is%20Valid
      (Niagara|SetNiagaraVariable(Bool) (Variables|/ChronoWeaponGameplay/Items/PanRifle/B_Assault_Chrono_Pan_Rifle2_Athena.B_Assault_Chrono_Pan_Rifle2_Athena_C|GetNS_FX_Barrel_Glow _asb_assault_chrono_pan_rifle_2_athena) "Fire" false)
      (return false))))
