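# Translucent unlit material for the PanRifle barrel glow effect, flagged for skeletal meshes, Niagara mesh particles and mesh deformers.
# An emissive color parameter times particle color feeds the shared emissive adjustments function.
# A gradient along local Y, measured from the particle position, is scaled, offset and raised to a power, then multiplied by opacity and both alphas.
# A scaled vertex-normal term provides a small push to inflate the shell.
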Material /ChronoWeaponGameplay/Items/PanRifle/FX/M_FX_PanRifle_BarrelGlow
graph {
 properties {
  bUsedWithMeshDeformer=true
  bUsedWithNiagaraMeshParticles=true
  bUsedWithSkeletalMesh=true
  blendMode=BLEND_Translucent
  pixelDepthOffsetMode=PDOM_Legacy
  shadingModel=MSM_Unlit
  usedShadingModels=Unlit
 }
 nodes [
  n1 = VectorParameter() {parameterName=Emissive Color, defaultValue=<rgba 10 2.0833 0.5 0.75>}
  n2 = ParticleColor()
  MaterialFunctionCall(Emissive=Multiply(A=@n1.RGB, B=@n2.RGB)) {materialFunction=&MF_FX_EmissiveAdjustments}
  Multiply(A=VertexNormalWS(), B=ScalarParameter() {defaultValue=0.075, parameterName=NormalPushAmount})
  n3 = ParticlePositionWS()
  n4 = ComponentMask(None=Subtract(
    A=MaterialFunctionCall() {materialFunction=&LocalPosition}.Local Position
    B=TransformPosition(None=@n3) {transformSourceType=TRANSFORMPOSSOURCE_World}
   )) {g=true}
  n5 = ScalarParameter() {defaultValue=15, parameterName=Divide}
  n6 = Power(
   Base=Add(A=Divide(A=@n4, B=@n5) {constB=150}, B=ScalarParameter() {parameterName=Offset}) {constB=0.06}
   Exp=ScalarParameter() {defaultValue=3, parameterName=Power}
  )
  n7 = ScalarParameter() {defaultValue=1, parameterName=Opacity}
  Multiply(A=@n1.A, B=Multiply(A=Multiply(A=@n6, B=@n7), B=@n2.A))
 ]
}
