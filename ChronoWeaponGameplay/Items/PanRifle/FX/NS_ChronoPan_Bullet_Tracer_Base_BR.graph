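# Niagara bullet tracer for the ChronoPan rifle, drawing stretched, oriented meshes from the muzzle toward hit points.
# One active emitter spawns a tracer per User.ImpactPositionArray entry; the other fires a single tracer to User.BeamEnd_0 when the array is empty.
# A frame-limited age drives travel, mesh stretch, bullet position and fades, with shot distance, camera LOD distance and time of day adjusting the look.
# FN_Bullet_TracerORIG is a disabled earlier variant.
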NiagaraSystem /ChronoWeaponGameplay/Items/PanRifle/FX/NS_ChronoPan_Bullet_Tracer_Base_BR
dependencies {
 usedDataInterfaces [&NiagaraDataInterfaceCurve, &NiagaraDataInterfaceVectorCurve]
 usedDynamicInputs [/Niagara/DynamicInputs/Math/GetVectorArrayCount, &NormalizeDistanceRange]
 usedModules [
  &self:SystemSpawnScript.SystemScriptSource.SystemScriptGraph.NiagaraNodeAssignment_2.SetVariables_#1
  &SystemState
  &self:SystemSpawnScript.SystemScriptSource.SystemScriptGraph.NiagaraNodeAssignment_0.SetVariables_#2
  &self:SystemSpawnScript.SystemScriptSource.SystemScriptGraph.NiagaraNodeAssignment_1.SetVariables_#3
 ]
}
emitters {
 FN_Bullet_Test_ArrayInput {
  emitterSpawnScript [/SetVariables(Emitter.BeamEnd:Vector3f=$User.BeamEnd_0, Emitter.DebugTimeMultiplier=1) {isSetParameters=true, module=SetVariables_#4}]
  emitterUpdateScript [
   &EmitterState()
   &SpawnBurst_Instantaneous(
    Spawn Count=&MakeCustomIntFromBool(
     Bool=&SetBoolByIntComparison(A=$System.ImpactArraySize, B=0, Comparison Type=::NewEnumerator0)
     True INT=$System.ImpactArraySize
     False INT=0
    )
    Use Spawn Probability=false
   )
  ]
  particleSpawnScript [
   &self:SetBeamEnd(New Vector Array:NiagaraDataInterfaceArrayFloat3=$User.ImpactPositionArray) {module=SetBeamEnd}
   /SetVariables(
    Particles.BeamTile:NiagaraFloat=&Divide_Float(
     A=/Niagara/DynamicInputs/TypeConversions/VectorLengthSafe(Vector:Vector3f=&Subtract_Vector(A=Linked() {linkedVariable {name=Engine.Owner.Position, type=NiagaraPosition}}, B=$Particles.BeamEndPosition))
     B=500
    )
    Particles.FireDirection:Vector3f=/Niagara/DynamicInputs/Helpers/SafeNormalizeVector(
     Vector To Normalize:Vector3f=&Subtract_Vector(A=$Particles.BeamEndPosition, B=Linked() {linkedVariable {name=Engine.Owner.Position, type=NiagaraPosition}})
     Fallback Vector:Vector3f=$Engine.Owner.SystemXAxis
    )
   ) {isSetParameters=true, module=SetVariables_#5}
   /SetVariables(
    Particles.BeamDistanceScalar:NiagaraFloat=&Lerp_Float(
     A=0.35
     B=1
     Alpha=&FloatFromCurve(
      FloatCurve=DataInterface() {propertyValues {Curve {keys [
          {arriveTangent=0.0002, interpMode=RCIM_Cubic, leaveTangent=0.0002, tangentMode=RCTM_User, time=0, value=0}
          {arriveTangent=0.0001, interpMode=RCIM_Cubic, leaveTangent=0.0001, tangentMode=RCTM_User, time=2500, value=0.35}
          <at 15000 0.8>
          {interpMode=RCIM_Cubic, tangentMode=RCTM_User, time=35000, value=1}
         ]}}}
      CurveIndex=&VectorLengthSafe(Vector:Vector3f=&Subtract_Vector(A=Linked() {linkedVariable {name=Engine.Owner.Position, type=NiagaraPosition}}, B=$Emitter.BeamEnd))
     )
    )
    Particles.LODDistanceRemap:NiagaraFloat=&FloatFromCurve(FloatCurve=DataInterface() {propertyValues {Curve {keys [<at 500 0>,<at 2500 1>]}}}, CurveIndex=$Engine.Owner.LODDistance)
   ) {isSetParameters=true, module=SetVariables_#6}
   &InitializeParticle(
    Color=<rgba 1 0.3543 0.026 0.9>
    Lifetime=&Multiply_Float(
     A=&Multiply_Float(
      A=&ClampFloat(Float=$Particles.BeamDistanceScalar, Min=0.5)
      B=&RandomRangeFloat(Minimum=0.7, Maximum=0.8)
     )
     B=$Emitter.DebugTimeMultiplier
    )
    Mesh Scale Mode=::NewEnumerator3
    UsePositionOffset=false
    Write Material Random=true
   )
   &DynamicMaterialParameters(
    Index 0 Param 4=&Multiply_Float(A=$Particles.BeamTile, B=&Lerp_Float(A=0.8, B=1.35, Alpha=$Particles.LODDistanceRemap))
    Index 1 Param 1=&RandomRangeFloat(Minimum=0.85, Maximum=1.05)
    Index 1 Param 2=&RandomRangeFloat(Minimum=0, Maximum=1)
    Module.Param0WriteEnabled=false
    Module.Param0WriteEnabled1=false
    Module.Param1WriteEnabled=false
    Module.Param1WriteEnabled1=true
    Module.Param2WriteEnabled=false
    Module.Param2WriteEnabled1=false
    Module.Param3WriteEnabled=true
    Module.Param3WriteEnabled1=false
    Write Parameter Index 1=true
   ) {module=DynamicMaterialParameters001}
   /Niagara/Modules/Update/Orientation/V2/OrientMeshToVector(Look At Direction:Vector3f=$Particles.FireDirection, Override Quat=false, UseOptionalSideAxis=false, UseOptionalUpAxis=false)
   /SetVariables(Particles.BulletDynParams:Vector4f=<1 1 1 1>) {isSetParameters=true, module=SetVariables_#7}
  ]
  particleUpdateScript [
   &ParticleState()
   /Niagara/Modules/Update/Utility/FrameCounter()
   /SetVariables(Particles.FrameLimitedAge:NiagaraFloat=&self:FN_Bullet_Test_ArrayInput_0.NiagaraScratchPadContainer_1.ScratchDynamicInput_5(
     MaxDuration:NiagaraFloat=0.5
     MinDuration:NiagaraFloat=0.1
     CurrentFrame:NiagaraFloat=&MakeFloatFromInt(Int=$Particles.FrameCounter.Counter)
     MaxFrames:NiagaraFloat=&MakeFloatFromInt(Int=8)
     DurationMultiplier:NiagaraFloat=$Particles.BeamDistanceScalar
     DebugTimeMultiplier:NiagaraFloat=$Emitter.DebugTimeMultiplier
    )) {isSetParameters=true, module=SetVariables_#8}
   /SetVariables(Particles.Travel:NiagaraFloat=&self:FN_Bullet_Test_ArrayInput_0.NiagaraScratchPadContainer_1.RemapRange_6(
     Value:NiagaraFloat=$Particles.FrameLimitedAge
     Input Min:NiagaraFloat=0
     Input Max:NiagaraFloat=0.5
     Output Min:NiagaraFloat=0
     Output Max:NiagaraFloat=1
    )) {isSetParameters=true, module=SetVariables_#9}
   /SetVariables(
    Particles.BulletPosition:Vector3f=&Lerp_Vector(
     A=&Add_Vector(
      A=Linked() {linkedVariable {name=Particles.Initial.Position, type=NiagaraPosition}}
      B=&Multiply_VectorByFloat(Vector=$Particles.FireDirection, Float=350)
     )
     B=$Particles.BeamEndPosition
     Alpha=$Particles.Travel
    )
    Particles.BulletVisbility:NiagaraInt32=&MakeCustomIntFromBool(
     Bool=&Boolean_AndOperation(
      Bool1=&SetBoolByFloatComparison(A=$Particles.FrameLimitedAge, B=0.55, Comparison Type=::NewEnumerator6)
      Bool2=&SetBoolByFloatComparison(A=$Particles.BeamTile, B=1)
     )
     True INT=0
     False INT=1
    )
   ) {isSetParameters=true, module=SetVariables_#10}
   /SetVariables(Particles.BulletSize:Vector3f=&Multiply_VectorByFloat(
     Vector=&Lerp_Vector(A=<2.5 1.25 1.25>, B=<4 2 2>, Alpha=$Particles.LODDistanceRemap)
     Float=&FloatFromCurve(
      FloatCurve=DataInterface() {propertyValues {Curve {keys [{arriveTangent=32.8701, time=0, value=0}, <at 0.05 1>]}}}
      CurveIndex=$Particles.FrameLimitedAge
     )
    )) {isSetParameters=true, module=SetVariables_#11}
   &ScaleMeshSize(Scale Factor=&MakeVector(
     X=&Lerp_Float(
      A=&Multiply_Float(A=$Particles.Travel, B=$Particles.BeamTile)
      B=100
      Alpha=&FloatFromCurve(
       FloatCurve=DataInterface() {propertyValues {Curve {keys [
           {interpMode=RCIM_Cubic, tangentMode=RCTM_Break, time=0.65, value=0}
           {arriveTangent=9.0989, interpMode=RCIM_Cubic, tangentMode=RCTM_Break, time=1, value=1}
          ]}}}
       CurveIndex=$System.BeamDistanceNormalized
      )
     )
     Y=1
     Z=1
    )) {module=ScaleMeshSize002}
   &ScaleMeshSize(Scale Factor=&MakeVector(X=&Multiply_Float(A=$Particles.Travel, B=$Particles.BeamTile), Y=1, Z=1)) {enabled=false}
   &DynamicMaterialParameters(
    Index 0 Param 1=&FloatFromCurve(FloatCurve=DataInterface() {propertyValues {Curve {keys [
         {arriveTangent=-0.4135, interpMode=RCIM_Cubic, leaveTangent=-0.4135, tangentMode=RCTM_User, time=0, value=1}
         {arriveTangent=-1.3177, interpMode=RCIM_Cubic, leaveTangent=-1.3177, tangentMode=RCTM_User, time=1, value=0}
        ]}}})
    Index 0 Param 2=&FloatFromCurve(
     FloatCurve=DataInterface() {propertyValues {Curve {keys [
         {arriveTangent=5.2627, interpMode=RCIM_Cubic, leaveTangent=5.2627, tangentMode=RCTM_User, time=0, value=0}
         {arriveTangent=0.8931, interpMode=RCIM_Cubic, leaveTangent=0.8931, tangentMode=RCTM_User, time=0.2287, value=0.7}
         {arriveTangentWeight=0.2571, interpMode=RCIM_Cubic, tangentMode=RCTM_User, time=1, value=1}
        ]}}}
     Scale Curve=&Multiply_Float(
      A=&Lerp_Float(A=0.5, B=1, Alpha=$Particles.BeamDistanceScalar)
      B=&RandomRangeFloat(Minimum=-10, Maximum=-8)
     )
    )
    Index 0 Param 3=&FloatFromCurve(
     FloatCurve=DataInterface() {propertyValues {Curve {keys [
         {arriveTangent=-0.9882, time=0.2, value=1}
         {arriveTangent=-1.2626, interpMode=RCIM_Cubic, leaveTangent=-1.2626, tangentMode=RCTM_User, time=0.5, value=0.6}
         {arriveTangent=-10.4271, interpMode=RCIM_Cubic, leaveTangent=-10.4271, tangentMode=RCTM_User, time=0.6, value=0}
        ]}}}
     CurveIndex=$Particles.FrameLimitedAge
    )
    Index 0 Param 4=&Multiply_Float(
     A=&MakeFloatFromVector(Vector=$Particles.Scale, Channel=::NewEnumerator0)
     B=&Lerp_Float(A=0.8, B=1.35, Alpha=$Particles.LODDistanceRemap)
    )
    Index 1 Param 1=&FloatFromCurve(
     FloatCurve=DataInterface() {propertyValues {Curve {keys [
         {arriveTangent=-0.8182, interpMode=RCIM_Cubic, leaveTangent=-0.8182, tangentMode=RCTM_User, time=0, value=0.7}
         {arriveTangent=2.4059, interpMode=RCIM_Cubic, leaveTangent=9.91, tangentMode=RCTM_Break, time=0.5, value=0.87}
         {arriveTangent=0.8994, interpMode=RCIM_Cubic, leaveTangent=0.8994, tangentMode=RCTM_User, time=0.5714, value=1.214}
         {arriveTangent=0.8128, interpMode=RCIM_Cubic, leaveTangent=0.8128, tangentMode=RCTM_User, time=1, value=1.5}
        ]}}}
     Scale Curve=&Lerp_Float(A=0.35, B=0.6, Alpha=$Particles.LODDistanceRemap)
    )
    Module.Param0WriteEnabled=true
    Module.Param0WriteEnabled1=true
    Module.Param1WriteEnabled=true
    Module.Param1WriteEnabled1=false
    Module.Param2WriteEnabled=true
    Module.Param2WriteEnabled1=false
    Module.Param3WriteEnabled=false
    Module.Param3WriteEnabled1=false
    Write Parameter Index 1=true
   )
   &ScaleColor(
    Scale Alpha=&FloatFromCurve(FloatCurve=DataInterface() {propertyValues {Curve {keys [
         {arriveTangent=-0.8527, interpMode=RCIM_Cubic, leaveTangent=-0.8527, tangentMode=RCTM_User, time=0.5, value=1}
         {arriveTangent=-1.0784, interpMode=RCIM_Cubic, leaveTangent=-1.0784, tangentMode=RCTM_User, time=0.52, value=0.6}
         {arriveTangent=-2.2297, interpMode=RCIM_Cubic, leaveTangent=-2.2297, tangentMode=RCTM_User, time=1, value=0}
        ]}}})
    Scale RGB=&VectorFromCurve(
     VectorCurve=DataInterface() {propertyValues {
       XCurve {keys [{arriveTangent=-4.22, interpMode=RCIM_Cubic, time=0.35, value=1}, {interpMode=RCIM_Cubic, leaveTangent=-7.2396, time=0.5, value=0}]}
       YCurve {keys [{arriveTangent=-8.4239, interpMode=RCIM_Cubic, time=0.35, value=1}, <at 0.5 0 RCIM_Cubic>]}
       ZCurve {keys [{arriveTangent=-8.2416, interpMode=RCIM_Cubic, time=0.35, value=1}, <at 0.5 0 RCIM_Cubic>]}
      }}
     CurveIndex=$Particles.FrameLimitedAge
     Scale Curve=&VectorFromFloat(Value=&FNTimeOfDayScalar(Night Scale=1, Morning Scale=1, Day Scale=1.5, Evening Scale=1))
    )
    ScaleA=true
    ScaleRGB=true
   )
  ]
  rendererClasses [&NiagaraMeshRendererProperties]
  renderers [
   {rendererClass=&NiagaraMeshRendererProperties, rendererIndex=0}
   {rendererClass=&NiagaraMeshRendererProperties, rendererIndex=1}
  ]
 }
 FN_Bullet_Tracer {
  emitterSpawnScript [/SetVariables(Emitter.BeamEnd:Vector3f=$User.BeamEnd_0, Emitter.DebugTimeMultiplier=1) {isSetParameters=true, module=SetVariables_#4}]
  emitterUpdateScript [
   &EmitterState()
   &SpawnBurst_Instantaneous(
    Spawn Count=&MakeCustomIntFromBool(
     Bool=&SetBoolByIntComparison(A=$System.ImpactArraySize, B=0, Comparison Type=::NewEnumerator0)
     True INT=0
     False INT=1
    )
    Use Spawn Probability=false
   )
  ]
  particleSpawnScript [
   /SetVariables(
    Particles.BeamTile:NiagaraFloat=&Divide_Float(
     A=&VectorLengthSafe(Vector:Vector3f=&Subtract_Vector(A=Linked() {linkedVariable {name=Engine.Owner.Position, type=NiagaraPosition}}, B=$Emitter.BeamEnd))
     B=500
    )
    Particles.FireDirection:Vector3f=&SafeNormalizeVector(
     Vector To Normalize:Vector3f=&Subtract_Vector(A=$Emitter.BeamEnd, B=Linked() {linkedVariable {name=Engine.Owner.Position, type=NiagaraPosition}})
     Fallback Vector:Vector3f=$Engine.Owner.SystemXAxis
    )
   ) {isSetParameters=true, module=SetVariables_#5}
   /SetVariables(
    Particles.BeamDistanceScalar:NiagaraFloat=&Lerp_Float(
     A=0.8
     B=1
     Alpha=&FloatFromCurve(
      FloatCurve=DataInterface() {propertyValues {Curve {keys [
          {arriveTangent=0.0002, interpMode=RCIM_Cubic, leaveTangent=0.0002, tangentMode=RCTM_User, time=0, value=0}
          {arriveTangent=0.0001, interpMode=RCIM_Cubic, leaveTangent=0.0001, tangentMode=RCTM_User, time=2500, value=0.35}
          <at 15000 0.8>
          {interpMode=RCIM_Cubic, tangentMode=RCTM_User, time=35000, value=1}
         ]}}}
      CurveIndex=&VectorLengthSafe(Vector:Vector3f=&Subtract_Vector(A=Linked() {linkedVariable {name=Engine.Owner.Position, type=NiagaraPosition}}, B=$Emitter.BeamEnd))
     )
    )
    Particles.LODDistanceRemap:NiagaraFloat=&FloatFromCurve(FloatCurve=DataInterface() {propertyValues {Curve {keys [<at 500 0>,<at 2500 1>]}}}, CurveIndex=$Engine.Owner.LODDistance)
   ) {isSetParameters=true, module=SetVariables_#6}
   &InitializeParticle(
    Color=<rgba 1 0.3543 0.026 0.9>
    Lifetime=&Multiply_Float(
     A=&Multiply_Float(
      A=&ClampFloat(Float=$Particles.BeamDistanceScalar, Min=0.5)
      B=&RandomRangeFloat(Minimum=0.7, Maximum=0.8)
     )
     B=$Emitter.DebugTimeMultiplier
    )
    Mesh Scale Mode=::NewEnumerator3
    UsePositionOffset=false
    Write Material Random=true
   )
   &DynamicMaterialParameters(
    Index 0 Param 4=&Multiply_Float(A=$Particles.BeamTile, B=&Lerp_Float(A=0.8, B=1.35, Alpha=$Particles.LODDistanceRemap))
    Index 1 Param 1=&RandomRangeFloat(Minimum=0.85, Maximum=1.05)
    Index 1 Param 2=&RandomRangeFloat(Minimum=0, Maximum=1)
    Module.Param0WriteEnabled=false
    Module.Param0WriteEnabled1=false
    Module.Param1WriteEnabled=false
    Module.Param1WriteEnabled1=true
    Module.Param2WriteEnabled=false
    Module.Param2WriteEnabled1=false
    Module.Param3WriteEnabled=true
    Module.Param3WriteEnabled1=false
    Write Parameter Index 1=true
   ) {module=DynamicMaterialParameters001}
   &OrientMeshToVector(Look At Direction:Vector3f=$Particles.FireDirection, Override Quat=false, UseOptionalSideAxis=false, UseOptionalUpAxis=false)
   /SetVariables(Particles.BulletDynParams:Vector4f=<1 1 1 1>) {isSetParameters=true, module=SetVariables_#7}
  ]
  particleUpdateScript [
   &ParticleState()
   &FrameCounter()
   /SetVariables(Particles.FrameLimitedAge:NiagaraFloat=&self:FN_Bullet_Tracer_1.NiagaraScratchPadContainer_1.ScratchDynamicInput_5(
     MaxDuration:NiagaraFloat=0.5
     MinDuration:NiagaraFloat=0.1
     CurrentFrame:NiagaraFloat=&MakeFloatFromInt(Int=$Particles.FrameCounter.Counter)
     MaxFrames:NiagaraFloat=&MakeFloatFromInt(Int=8)
     DurationMultiplier:NiagaraFloat=$Particles.BeamDistanceScalar
     DebugTimeMultiplier:NiagaraFloat=$Emitter.DebugTimeMultiplier
    )) {isSetParameters=true, module=SetVariables_#8}
   /SetVariables(Particles.Travel:NiagaraFloat=&self:FN_Bullet_Tracer_1.NiagaraScratchPadContainer_1.RemapRange_6(
     Value:NiagaraFloat=$Particles.FrameLimitedAge
     Input Min:NiagaraFloat=0
     Input Max:NiagaraFloat=0.5
     Output Min:NiagaraFloat=0
     Output Max:NiagaraFloat=1
    )) {isSetParameters=true, module=SetVariables_#9}
   /SetVariables(
    Particles.BulletPosition:Vector3f=&Lerp_Vector(
     A=&Add_Vector(
      A=Linked() {linkedVariable {name=Particles.Initial.Position, type=NiagaraPosition}}
      B=&Multiply_VectorByFloat(Vector=$Particles.FireDirection, Float=350)
     )
     B=$Emitter.BeamEnd
     Alpha=$Particles.Travel
    )
    Particles.BulletVisbility:NiagaraInt32=&MakeCustomIntFromBool(
     Bool=&Boolean_AndOperation(
      Bool1=&SetBoolByFloatComparison(A=$Particles.FrameLimitedAge, B=0.55, Comparison Type=::NewEnumerator6)
      Bool2=&SetBoolByFloatComparison(A=$Particles.BeamTile, B=1)
     )
     True INT=0
     False INT=1
    )
   ) {isSetParameters=true, module=SetVariables_#10}
   /SetVariables(Particles.BulletSize:Vector3f=&Multiply_VectorByFloat(
     Vector=&Lerp_Vector(A=<2.5 1.25 1.25>, B=<4 2 2>, Alpha=$Particles.LODDistanceRemap)
     Float=&FloatFromCurve(
      FloatCurve=DataInterface() {propertyValues {Curve {keys [{arriveTangent=32.8701, time=0, value=0}, <at 0.05 1>]}}}
      CurveIndex=$Particles.FrameLimitedAge
     )
    )) {isSetParameters=true, module=SetVariables_#11}
   &ScaleMeshSize(Scale Factor=&MakeVector(
     X=&Lerp_Float(
      A=&Multiply_Float(A=$Particles.Travel, B=$Particles.BeamTile)
      B=100
      Alpha=&FloatFromCurve(
       FloatCurve=DataInterface() {propertyValues {Curve {keys [
           {interpMode=RCIM_Cubic, tangentMode=RCTM_Break, time=0.65, value=0}
           {arriveTangent=9.0989, interpMode=RCIM_Cubic, tangentMode=RCTM_Break, time=1, value=1}
          ]}}}
       CurveIndex=$System.BeamDistanceNormalized
      )
     )
     Y=1
     Z=1
    )) {module=ScaleMeshSize002}
   &ScaleMeshSize(Scale Factor=&MakeVector(X=&Multiply_Float(A=$Particles.Travel, B=$Particles.BeamTile), Y=1, Z=1)) {enabled=false}
   &DynamicMaterialParameters(
    Index 0 Param 1=&FloatFromCurve(FloatCurve=DataInterface() {propertyValues {Curve {keys [
         {arriveTangent=-0.4135, interpMode=RCIM_Cubic, leaveTangent=-0.4135, tangentMode=RCTM_User, time=0, value=1}
         {arriveTangent=-1.3177, interpMode=RCIM_Cubic, leaveTangent=-1.3177, tangentMode=RCTM_User, time=1, value=0}
        ]}}})
    Index 0 Param 2=&FloatFromCurve(
     FloatCurve=DataInterface() {propertyValues {Curve {keys [
         {arriveTangent=5.2627, interpMode=RCIM_Cubic, leaveTangent=5.2627, tangentMode=RCTM_User, time=0, value=0}
         {arriveTangent=0.8931, interpMode=RCIM_Cubic, leaveTangent=0.8931, tangentMode=RCTM_User, time=0.2287, value=0.7}
         {arriveTangentWeight=0.2571, interpMode=RCIM_Cubic, tangentMode=RCTM_User, time=1, value=1}
        ]}}}
     Scale Curve=&Multiply_Float(
      A=&Lerp_Float(A=0.5, B=1, Alpha=$Particles.BeamDistanceScalar)
      B=&RandomRangeFloat(Minimum=-10, Maximum=-8)
     )
    )
    Index 0 Param 3=&FloatFromCurve(
     FloatCurve=DataInterface() {propertyValues {Curve {keys [
         {arriveTangent=-0.9882, time=0.2, value=1}
         {arriveTangent=-1.2626, interpMode=RCIM_Cubic, leaveTangent=-1.2626, tangentMode=RCTM_User, time=0.5, value=0.6}
         {arriveTangent=-10.4271, interpMode=RCIM_Cubic, leaveTangent=-10.4271, tangentMode=RCTM_User, time=0.6, value=0}
        ]}}}
     CurveIndex=$Particles.FrameLimitedAge
    )
    Index 0 Param 4=&Multiply_Float(
     A=&MakeFloatFromVector(Vector=$Particles.Scale, Channel=::NewEnumerator0)
     B=&Lerp_Float(A=0.8, B=1.35, Alpha=$Particles.LODDistanceRemap)
    )
    Index 1 Param 1=&FloatFromCurve(
     FloatCurve=DataInterface() {propertyValues {Curve {keys [
         {arriveTangent=-0.8182, interpMode=RCIM_Cubic, leaveTangent=-0.8182, tangentMode=RCTM_User, time=0, value=0.7}
         {arriveTangent=2.4059, interpMode=RCIM_Cubic, leaveTangent=9.91, tangentMode=RCTM_Break, time=0.5, value=0.87}
         {arriveTangent=0.8994, interpMode=RCIM_Cubic, leaveTangent=0.8994, tangentMode=RCTM_User, time=0.5714, value=1.214}
         {arriveTangent=0.8128, interpMode=RCIM_Cubic, leaveTangent=0.8128, tangentMode=RCTM_User, time=1, value=1.5}
        ]}}}
     Scale Curve=&Lerp_Float(A=1, B=1.2, Alpha=$Particles.LODDistanceRemap)
    )
    Module.Param0WriteEnabled=true
    Module.Param0WriteEnabled1=true
    Module.Param1WriteEnabled=true
    Module.Param1WriteEnabled1=false
    Module.Param2WriteEnabled=true
    Module.Param2WriteEnabled1=false
    Module.Param3WriteEnabled=false
    Module.Param3WriteEnabled1=false
    Write Parameter Index 1=true
   )
   &ScaleColor(
    Scale Alpha=&FloatFromCurve(FloatCurve=DataInterface() {propertyValues {Curve {keys [
         {arriveTangent=-0.8527, interpMode=RCIM_Cubic, leaveTangent=-0.8527, tangentMode=RCTM_User, time=0.5, value=1}
         {arriveTangent=-1.0784, interpMode=RCIM_Cubic, leaveTangent=-1.0784, tangentMode=RCTM_User, time=0.52, value=0.6}
         {arriveTangent=-2.2297, interpMode=RCIM_Cubic, leaveTangent=-2.2297, tangentMode=RCTM_User, time=1, value=0}
        ]}}})
    Scale RGB=&VectorFromCurve(
     VectorCurve=DataInterface() {propertyValues {
       XCurve {keys [{arriveTangent=-4.22, interpMode=RCIM_Cubic, time=0.35, value=1}, {interpMode=RCIM_Cubic, leaveTangent=-7.2396, time=0.5, value=0}]}
       YCurve {keys [{arriveTangent=-8.4239, interpMode=RCIM_Cubic, time=0.35, value=1}, <at 0.5 0 RCIM_Cubic>]}
       ZCurve {keys [{arriveTangent=-8.2416, interpMode=RCIM_Cubic, time=0.35, value=1}, <at 0.5 0 RCIM_Cubic>]}
      }}
     CurveIndex=$Particles.FrameLimitedAge
     Scale Curve=&VectorFromFloat(Value=&FNTimeOfDayScalar(Night Scale=1, Morning Scale=1, Day Scale=1.5, Evening Scale=1))
    )
    ScaleA=true
    ScaleRGB=true
   )
  ]
  rendererClasses [&NiagaraMeshRendererProperties]
  renderers [
   {rendererClass=&NiagaraMeshRendererProperties, rendererIndex=0}
   {rendererClass=&NiagaraMeshRendererProperties, rendererIndex=1}
  ]
 }
 FN_Bullet_TracerORIG {
  bEnabled=false
  emitterSpawnScript [/SetVariables(Emitter.BeamEnd:Vector3f=$User.BeamEnd_0, Emitter.DebugTimeMultiplier=1) {isSetParameters=true, module=SetVariables_#4}]
  emitterUpdateScript [
   &EmitterState()
   &SpawnBurst_Instantaneous(Spawn Count=1, Use Spawn Probability=false)
  ]
  particleSpawnScript [
   /SetVariables(
    Particles.BeamTile:NiagaraFloat=&Divide_Float(
     A=&VectorLengthSafe(Vector:Vector3f=&Subtract_Vector(A=Linked() {linkedVariable {name=Engine.Owner.Position, type=NiagaraPosition}}, B=$Emitter.BeamEnd))
     B=500
    )
    Particles.FireDirection:Vector3f=&SafeNormalizeVector(
     Vector To Normalize:Vector3f=&Subtract_Vector(A=$Emitter.BeamEnd, B=Linked() {linkedVariable {name=Engine.Owner.Position, type=NiagaraPosition}})
     Fallback Vector:Vector3f=$Engine.Owner.SystemXAxis
    )
   ) {isSetParameters=true, module=SetVariables_#5}
   /SetVariables(
    Particles.BeamDistanceScalar:NiagaraFloat=&Lerp_Float(
     A=0.35
     B=1
     Alpha=&FloatFromCurve(
      FloatCurve=DataInterface() {propertyValues {Curve {keys [
          {arriveTangent=0.0002, interpMode=RCIM_Cubic, leaveTangent=0.0002, tangentMode=RCTM_User, time=0, value=0}
          {arriveTangent=0.0001, interpMode=RCIM_Cubic, leaveTangent=0.0001, tangentMode=RCTM_User, time=2500, value=0.35}
          <at 15000 0.8>
          {interpMode=RCIM_Cubic, tangentMode=RCTM_User, time=35000, value=1}
         ]}}}
      CurveIndex=&VectorLengthSafe(Vector:Vector3f=&Subtract_Vector(A=Linked() {linkedVariable {name=Engine.Owner.Position, type=NiagaraPosition}}, B=$Emitter.BeamEnd))
     )
    )
    Particles.LODDistanceRemap:NiagaraFloat=&FloatFromCurve(FloatCurve=DataInterface() {propertyValues {Curve {keys [<at 500 0>,<at 2500 1>]}}}, CurveIndex=$Engine.Owner.LODDistance)
   ) {isSetParameters=true, module=SetVariables_#6}
   &InitializeParticle(
    Color=<rgba 1 0.3543 0.026 0.9>
    Lifetime=&Multiply_Float(
     A=&Multiply_Float(
      A=&ClampFloat(Float=$Particles.BeamDistanceScalar, Min=0.5)
      B=&RandomRangeFloat(Minimum=0.7, Maximum=0.8)
     )
     B=$Emitter.DebugTimeMultiplier
    )
    Mesh Scale Mode=::NewEnumerator3
    UsePositionOffset=false
    Write Material Random=true
   )
   &DynamicMaterialParameters(
    Index 0 Param 4=&Multiply_Float(A=$Particles.BeamTile, B=&Lerp_Float(A=0.8, B=1.35, Alpha=$Particles.LODDistanceRemap))
    Index 1 Param 1=&RandomRangeFloat(Minimum=0.85, Maximum=1.05)
    Index 1 Param 2=&RandomRangeFloat(Minimum=0, Maximum=1)
    Module.Param0WriteEnabled=false
    Module.Param0WriteEnabled1=false
    Module.Param1WriteEnabled=false
    Module.Param1WriteEnabled1=true
    Module.Param2WriteEnabled=false
    Module.Param2WriteEnabled1=false
    Module.Param3WriteEnabled=true
    Module.Param3WriteEnabled1=false
    Write Parameter Index 1=true
   ) {module=DynamicMaterialParameters001}
   &OrientMeshToVector(Look At Direction:Vector3f=$Particles.FireDirection, Override Quat=false, UseOptionalSideAxis=false, UseOptionalUpAxis=false)
   /SetVariables(Particles.BulletDynParams:Vector4f=<1 1 1 1>) {isSetParameters=true, module=SetVariables_#7}
  ]
  particleUpdateScript [
   &ParticleState()
   &FrameCounter()
   /SetVariables(Particles.FrameLimitedAge:NiagaraFloat=&self:FN_Bullet_TracerORIG_0.NiagaraScratchPadContainer_1.ScratchDynamicInput_5(
     MaxDuration:NiagaraFloat=0.5
     MinDuration:NiagaraFloat=0.1
     CurrentFrame:NiagaraFloat=&MakeFloatFromInt(Int=$Particles.FrameCounter.Counter)
     MaxFrames:NiagaraFloat=&MakeFloatFromInt(Int=14)
     DurationMultiplier:NiagaraFloat=$Particles.BeamDistanceScalar
     DebugTimeMultiplier:NiagaraFloat=$Emitter.DebugTimeMultiplier
    )) {isSetParameters=true, module=SetVariables_#8}
   /SetVariables(Particles.Travel:NiagaraFloat=&self:FN_Bullet_TracerORIG_0.NiagaraScratchPadContainer_1.RemapRange_6(
     Value:NiagaraFloat=$Particles.FrameLimitedAge
     Input Min:NiagaraFloat=0
     Input Max:NiagaraFloat=0.5
     Output Min:NiagaraFloat=0.01
     Output Max:NiagaraFloat=1
    )) {isSetParameters=true, module=SetVariables_#9}
   /SetVariables(
    Particles.BulletPosition:Vector3f=&Lerp_Vector(
     A=&Add_Vector(
      A=Linked() {linkedVariable {name=Particles.Initial.Position, type=NiagaraPosition}}
      B=&Multiply_VectorByFloat(Vector=$Particles.FireDirection, Float=350)
     )
     B=$Emitter.BeamEnd
     Alpha=$Particles.Travel
    )
    Particles.BulletVisbility:NiagaraInt32=&MakeCustomIntFromBool(
     Bool=&Boolean_AndOperation(
      Bool1=&SetBoolByFloatComparison(A=$Particles.FrameLimitedAge, B=0.55, Comparison Type=::NewEnumerator6)
      Bool2=&SetBoolByFloatComparison(A=$Particles.BeamTile, B=1)
     )
     True INT=0
     False INT=1
    )
   ) {isSetParameters=true, module=SetVariables_#10}
   /SetVariables(Particles.BulletSize:Vector3f=&Multiply_VectorByFloat(
     Vector=&Lerp_Vector(A=<2.5 1.25 1.25>, B=<4 2 2>, Alpha=$Particles.LODDistanceRemap)
     Float=&FloatFromCurve(
      FloatCurve=DataInterface() {propertyValues {Curve {keys [{arriveTangent=32.8701, time=0, value=0}, <at 0.05 1>]}}}
      CurveIndex=$Particles.FrameLimitedAge
     )
    )) {isSetParameters=true, module=SetVariables_#11}
   &ScaleMeshSize(Scale Factor=&MakeVector(X=&Multiply_Float(A=$Particles.Travel, B=$Particles.BeamTile), Y=1, Z=1))
   &DynamicMaterialParameters(
    Index 0 Param 1=&FloatFromCurve(FloatCurve=DataInterface() {propertyValues {Curve {keys [
         {arriveTangent=-0.4135, interpMode=RCIM_Cubic, leaveTangent=-0.4135, tangentMode=RCTM_User, time=0, value=1}
         {arriveTangent=-1.3177, interpMode=RCIM_Cubic, leaveTangent=-1.3177, tangentMode=RCTM_User, time=1, value=0}
        ]}}})
    Index 0 Param 2=&FloatFromCurve(
     FloatCurve=DataInterface() {propertyValues {Curve {keys [
         {arriveTangent=5.2627, interpMode=RCIM_Cubic, leaveTangent=5.2627, tangentMode=RCTM_User, time=0, value=0}
         {arriveTangent=0.8931, interpMode=RCIM_Cubic, leaveTangent=0.8931, tangentMode=RCTM_User, time=0.2287, value=0.7}
         {arriveTangentWeight=0.2571, interpMode=RCIM_Cubic, tangentMode=RCTM_User, time=1, value=1}
        ]}}}
     Scale Curve=&Multiply_Float(
      A=&Lerp_Float(A=0.5, B=1, Alpha=$Particles.BeamDistanceScalar)
      B=&RandomRangeFloat(Minimum=-10, Maximum=-8)
     )
    )
    Index 0 Param 3=&FloatFromCurve(
     FloatCurve=DataInterface() {propertyValues {Curve {keys [
         {arriveTangent=-0.9882, time=0.2, value=1}
         {arriveTangent=-1.2626, interpMode=RCIM_Cubic, leaveTangent=-1.2626, tangentMode=RCTM_User, time=0.5, value=0.6}
         {arriveTangent=-10.4271, interpMode=RCIM_Cubic, leaveTangent=-10.4271, tangentMode=RCTM_User, time=0.6, value=0}
        ]}}}
     CurveIndex=$Particles.FrameLimitedAge
    )
    Index 0 Param 4=&Multiply_Float(
     A=&MakeFloatFromVector(Vector=$Particles.Scale, Channel=::NewEnumerator0)
     B=&Lerp_Float(A=0.8, B=1.35, Alpha=$Particles.LODDistanceRemap)
    )
    Index 1 Param 1=&FloatFromCurve(
     FloatCurve=DataInterface() {propertyValues {Curve {keys [
         {arriveTangent=-0.8182, interpMode=RCIM_Cubic, leaveTangent=-0.8182, tangentMode=RCTM_User, time=0, value=0.7}
         {arriveTangent=2.4059, interpMode=RCIM_Cubic, leaveTangent=9.91, tangentMode=RCTM_Break, time=0.5, value=0.87}
         {arriveTangent=0.8994, interpMode=RCIM_Cubic, leaveTangent=0.8994, tangentMode=RCTM_User, time=0.5714, value=1.214}
         {arriveTangent=0.8128, interpMode=RCIM_Cubic, leaveTangent=0.8128, tangentMode=RCTM_User, time=1, value=1.5}
        ]}}}
     Scale Curve=&Lerp_Float(A=0.35, B=0.6, Alpha=$Particles.LODDistanceRemap)
    )
    Module.Param0WriteEnabled=true
    Module.Param0WriteEnabled1=true
    Module.Param1WriteEnabled=true
    Module.Param1WriteEnabled1=false
    Module.Param2WriteEnabled=true
    Module.Param2WriteEnabled1=false
    Module.Param3WriteEnabled=false
    Module.Param3WriteEnabled1=false
    Write Parameter Index 1=true
   )
   &ScaleColor(
    Scale Alpha=&FloatFromCurve(FloatCurve=DataInterface() {propertyValues {Curve {keys [
         {arriveTangent=-0.8527, interpMode=RCIM_Cubic, leaveTangent=-0.8527, tangentMode=RCTM_User, time=0.5, value=1}
         {arriveTangent=-1.0784, interpMode=RCIM_Cubic, leaveTangent=-1.0784, tangentMode=RCTM_User, time=0.52, value=0.6}
         {arriveTangent=-2.2297, interpMode=RCIM_Cubic, leaveTangent=-2.2297, tangentMode=RCTM_User, time=1, value=0}
        ]}}})
    Scale RGB=&VectorFromCurve(
     VectorCurve=DataInterface() {propertyValues {
       XCurve {keys [{arriveTangent=-4.22, interpMode=RCIM_Cubic, time=0.35, value=1}, {interpMode=RCIM_Cubic, leaveTangent=-7.2396, time=0.6, value=0}]}
       YCurve {keys [{arriveTangent=-8.4239, interpMode=RCIM_Cubic, time=0.35, value=1}, <at 0.6 0 RCIM_Cubic>]}
       ZCurve {keys [{arriveTangent=-8.2416, interpMode=RCIM_Cubic, time=0.35, value=1}, <at 0.6 0 RCIM_Cubic>]}
      }}
     CurveIndex=$Particles.FrameLimitedAge
    )
    ScaleA=true
    ScaleRGB=true
   )
  ]
  rendererClasses [&NiagaraMeshRendererProperties]
  renderers [
   {rendererClass=&NiagaraMeshRendererProperties, rendererIndex=0}
   {rendererClass=&NiagaraMeshRendererProperties, rendererIndex=1}
  ]
 }
}
summary {
 systemName=NS_ChronoPan_Bullet_Tracer_Base_BR
 emitters [
  {emitterName=FN_Bullet_Tracer, rendererClasses [&NiagaraMeshRendererProperties]}
  {bEnabled=false, emitterName=FN_Bullet_TracerORIG, rendererClasses [&NiagaraMeshRendererProperties]}
  {emitterName=FN_Bullet_Test_ArrayInput, rendererClasses [&NiagaraMeshRendererProperties]}
 ]
 userVariables [
  User.BeamEnd_0 {type=Vector3f, description="", defaultValue {struct=Vector3f, value=<0 0 2000>}}
  User.ImpactPositionArray {
   type=NiagaraDataInterfaceArrayFloat3
   description=""
   defaultValue {
    struct=NiagaraExt_VariableValue_DataInterface
    value {
     dataInterface=&self:NiagaraDataInterfaceArrayFloat3_0
     dataInterfaceClass=&NiagaraDataInterfaceArrayFloat3
    }
   }
  }
  User.MaxDistance {type=NiagaraFloat, defaultValue=27000, description=""}
  User.UseArrayForEndPosition {type=NiagaraBool}
 ]
}
