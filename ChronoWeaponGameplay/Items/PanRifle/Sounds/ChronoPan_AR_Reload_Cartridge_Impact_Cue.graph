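# Sound cue for the ChronoPan rifle's reload, supplying the cartridge impact sound.
# A Random node at the root picks among three cartridge impact wave variations with equal weight, so repeated reloads can sound different.
# The waves play directly through the Random node, with no modulation, mixing or attenuation set in the cue.
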SoundCue /ChronoWeaponGameplay/Items/PanRifle/Sounds/ChronoPan_AR_Reload_Cartridge_Impact_Cue
root=Random_0
nodes [
 {id=Random_0, children [WavePlayer_0,WavePlayer_1,WavePlayer_2], properties {weights [1,1,1]}}
 {id=WavePlayer_0, properties {soundWaveAssetPtr=&/ChronoPan_AR_Reload_Cartridge_Impact_01}}
 {id=WavePlayer_1, properties {soundWaveAssetPtr=&/ChronoPan_AR_Reload_Cartridge_Impact_02}}
 {id=WavePlayer_2, properties {soundWaveAssetPtr=&/ChronoPan_AR_Reload_Cartridge_Impact_03}}
]
properties {maxDistance=2097152}
