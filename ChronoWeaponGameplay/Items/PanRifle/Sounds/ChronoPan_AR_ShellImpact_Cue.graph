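# Sound cue that plays shell impact sounds for the ChronoPan assault rifle (Pan Rifle).
# A quality-level node picks an equally weighted random pool: six wave variants for the first two levels, three for the last two.
# A modulator at the root wraps the selection.
# The cue plays at reduced volume, borrowing weapon-reload attenuation and a third-person ranged-weapon-fire sound class.
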SoundCue /ChronoWeaponGameplay/Items/PanRifle/Sounds/ChronoPan_AR_ShellImpact_Cue
root=Modulator_0
nodes [
 {id=Modulator_0, children [QualityLevel_0]}
 {id=QualityLevel_0, children [Random_0,Random_0,Random_1,Random_1]}
 {id=Random_0, children [WavePlayer_0,WavePlayer_1,WavePlayer_2,WavePlayer_3,WavePlayer_4,WavePlayer_5], properties {weights [1,1,1,1,1,1]}}
 {id=Random_1, children [WavePlayer_6,WavePlayer_7,WavePlayer_8], properties {weights [1,1,1]}}
 {id=WavePlayer_0, properties {soundWaveAssetPtr=&/ChronoPan_AR_ShellImpact_01}}
 {id=WavePlayer_1, properties {soundWaveAssetPtr=&/ChronoPan_AR_ShellImpact_02}}
 {id=WavePlayer_2, properties {soundWaveAssetPtr=&/ChronoPan_AR_ShellImpact_03}}
 {id=WavePlayer_3, properties {soundWaveAssetPtr=&/ChronoPan_AR_ShellImpact_04}}
 {id=WavePlayer_4, properties {soundWaveAssetPtr=&/ChronoPan_AR_ShellImpact_05}}
 {id=WavePlayer_5, properties {soundWaveAssetPtr=&/ChronoPan_AR_ShellImpact_06}}
 {id=WavePlayer_6, properties {soundWaveAssetPtr=&ChronoPan_AR_ShellImpact_01}}
 {id=WavePlayer_7, properties {soundWaveAssetPtr=&ChronoPan_AR_ShellImpact_02}}
 {id=WavePlayer_8, properties {soundWaveAssetPtr=&ChronoPan_AR_ShellImpact_03}}
]
properties {
 attenuationSettings=&Fort_Weapon_Reload_Att
 maxDistance=2250
 soundClassObject=&Fort_WeaponRanged_Fire_3P_PSM
 volumeMultiplier=0.4
}
