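# Niagara system that ejects shell casings for the PanRifle's looping fire effect.
# A probabilistic one-particle burst spawns each casing in a cylinder, with User parameters driving velocity, gravity, inherited motion, spin, size and color.
# User Shell Type sets the visibility tag across five mesh renderers and feeds the impact sound cue.
# Collision, gated by owner LOD distance and particle age, triggers persistent impact audio; casings shrink out at end of life.
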NiagaraSystem /ChronoWeaponGameplay/Items/PanRifle/FX/NS_PanRifle_ShellsFire_Looping
dependencies {
 usedDataInterfaces [&NiagaraDataInterfaceAudioPlayer, &NiagaraDataInterfaceCurve]
 usedDynamicInputs [&Divide_Float]
 usedModules [
  &SystemState
  &self:SystemSpawnScript.SystemScriptSource.SystemScriptGraph.NiagaraNodeAssignment_0.SetVariables_#1
 ]
}
emitters {Shell_Eject_Looping {
  emitterSpawnScript [/SetVariables(Emitter.Audio Player=DataInterface() {propertyValues {
      Attenuation=null
      Concurrency=null
      MaxPlaysPerTick=10
      SoundToPlay=/ChronoWeaponGameplay/Items/PanRifle/Sounds/ChronoPan_AR_ShellImpact_Cue
      bAllowLoopingOneShotSounds=false
      bLimitPlaysPerTick=true
      bOnlyActiveDuringGameplay=false
      bStopWhenComponentIsDestroyed=true
      ConfigurationUserParameter {parameter {name=None, typeDefHandle {registeredTypeIndex=1}, varData []}}
      ParameterNames [ShellType]
     }}) {isSetParameters=true}]
  emitterUpdateScript [
   &EmitterState()
   &SpawnBurst_Instantaneous(Spawn Count=1, Spawn Probability=$User.SpawnProbability, Use Spawn Probability=true)
  ]
  particleSpawnScript [
   /Niagara/Modules/Spawn/Initialization/InitializeParticle(
    Lifetime=1.5
    Module.Write Color=true
    Module.Write Lifetime=true
    Module.Write Mass=true
    Module.Write Material Random=true
    Module.Write Position=false
    Module.Write Scale=true
    Module.Write SpriteRotation=false
    Module.Write SpriteSize=false
    Sprite Size=<5 5>
   )
   /SetVariables(Particles.VisibilityTag=$User.Shell Type) {isSetParameters=true}
   /SetVariables(Particles.UniformMeshOrientation_Initial:Vector3f=&RandomRangeVector(Minimum=<0.2 0.2 0.2>, Maximum=<1 1 1>)) {isSetParameters=true}
   &CylinderLocation(
    Cylinder Height:NiagaraFloat=$User.CylinderHeight
    Cylinder Origin:Vector3f=/Niagara/DynamicInputs/Helpers/SimulationPosition()
    Cylinder Radius=$User.CylinderRadius
    MaskSpawnGroup=false
    Module.Enable Lathe Profile=false
    Module.Override Randomness=false
    Module.Surface Only=false
    Orientation Axis:ENiagaraOrientationAxis=::XAxis
   )
   &AddVelocity(Velocity=$User.Velocity)
   /SetVariables() {isSetParameters=true}
  ]
  particleUpdateScript [
   &ParticleState()
   &AccelerationForce(Acceleration=$User.Gravity)
   &InheritVelocity(Module.Limit Speed=true, Velocity Limit=10, Velocity Scale=$User.InheritParentVelocity)
   &Collision(
    Age Colliding Particles=false
    Collision Enabled=&Boolean_AndOperation(
     Bool1=&SetBoolByFloatComparison(A=$Engine.Owner.LODDistance, B=1500, Comparison Type=::NewEnumerator6)
     Bool2=&SetBoolByFloatComparison(A=$Particles.Age, B=0.2)
    )
    Friction=0.7
    Module.Control Rest State Via Surface Normal=false
    Module.EnableMaxCollisionCount=false
    Module.Randomize Collision Normal Vector=true
    Radius Calculation Type=::NewEnumerator1
    Randomize Collision Normal=0.12
    Restitution=0.4
   )
   /SetVariables(Particles.PlayCollisionSound:NiagaraBool=&Boolean_OrOperation(Bool1=$Particles.PlayCollisionSound, Bool2=$Particles.HasCollided)) {isSetParameters=true}
   /Niagara/Modules/Audio/PlayAudio(Audio Player:NiagaraDataInterfaceAudioPlayer=$Emitter.Audio Player, Play Audio Trigger:NiagaraBool=$Particles.PlayCollisionSound) {enabled=false}
   /Niagara/Modules/Audio/PlayPersistentAudio(
    Audio Player:NiagaraDataInterfaceAudioPlayer=$Emitter.Audio Player
    Enable Fade In / Out=true
    Fade In Time=0
    Fade Out Time=1.5
    Play Audio:NiagaraBool=$Particles.PlayCollisionSound
   )
   /Niagara/Modules/Audio/UpdatePersistentAudio(
    Audio Player:NiagaraDataInterfaceAudioPlayer=$Emitter.Audio Player
    Int Value:NiagaraInt32=$User.Shell Type
    Update Type:ENiagara_AudioParamType=::NewEnumerator1
   ) {module=UpdatePersistentAudio002}
   &UpdatePersistentAudio(Audio Player:NiagaraDataInterfaceAudioPlayer=$Emitter.Audio Player, Update Type:ENiagara_AudioParamType=::NewEnumerator5)
   &UpdatePersistentAudio(
    Audio Player:NiagaraDataInterfaceAudioPlayer=$Emitter.Audio Player
    Volume:NiagaraFloat=&FloatFromCurve(FloatCurve=DataInterface() {propertyValues {Curve {keys [<at 0 1>,<at 0.9 1>,<at 1 0>]}}})
   ) {module=UpdatePersistentAudio001}
   &SolveForcesAndVelocity(Clamp Velocity=false, Limit Acceleration=false)
   &ScaleMeshSize(Scale Factor=$User.Size)
   &ScaleMeshSize(Scale Factor=&VectorFromFloat(Value=&FloatFromCurve(FloatCurve=DataInterface() {propertyValues {Curve {keys [<at 0.7517 1 RCIM_Cubic>, {interpMode=RCIM_Cubic, tangentMode=RCTM_User, time=0.8642, value=1}, <at 1 0 RCIM_Cubic>]}}}))) {module=ScaleMeshSize001}
   &ScaleMeshSize(Scale Factor=&VectorFromFloat(Value=&FloatFromCurve(FloatCurve=DataInterface() {propertyValues {Curve {keys [<at 0.7517 1 RCIM_Cubic>, {interpMode=RCIM_Cubic, tangentMode=RCTM_User, time=0.8642, value=1.3298}, <at 1 0 RCIM_Cubic>]}}}))) {enabled=false, module=ScaleMeshSize002}
   /Niagara/Modules/Update/Orientation/V2/UpdateMeshOrientation(Rotation Vector:Vector3f=&Multiply_Vector(A=$Particles.UniformMeshOrientation_Initial, B=$User.ShellsRotationRate))
   &DynamicMaterialParameters(
    Index 0 Param 1=&MakeFloatFromLinearColor(LinearColor=$User.ShellColor, Channel:ENiagaraLinearColor_Channels=::NewEnumerator0)
    Index 0 Param 2=&MakeFloatFromLinearColor(LinearColor=$User.ShellColor, Channel:ENiagaraLinearColor_Channels=::NewEnumerator1)
    Index 0 Param 3=&MakeFloatFromLinearColor(LinearColor=$User.ShellColor, Channel:ENiagaraLinearColor_Channels=::NewEnumerator2)
    Module.Param0WriteEnabled=true
    Module.Param1WriteEnabled=true
    Module.Param2WriteEnabled=true
    Module.Param3WriteEnabled=false
   )
   /SetVariables(Particles.DEBUG_ShellType:NiagaraInt32=$User.Shell Type) {isSetParameters=true}
  ]
  rendererClasses [&NiagaraMeshRendererProperties, &NiagaraSpriteRendererProperties]
  renderers [
   {rendererClass=&NiagaraMeshRendererProperties, rendererIndex=0}
   {rendererClass=&NiagaraMeshRendererProperties, rendererIndex=1}
   {rendererClass=&NiagaraMeshRendererProperties, rendererIndex=2}
   {rendererClass=&NiagaraMeshRendererProperties, rendererIndex=3}
   {rendererClass=&NiagaraMeshRendererProperties, rendererIndex=4}
   {rendererClass=&NiagaraSpriteRendererProperties, rendererIndex=5}
  ]
 }}
summary {
 systemName=NS_PanRifle_ShellsFire_Looping
 emitters [{
   emitterName=Shell_Eject_Looping
   rendererClasses [&NiagaraMeshRendererProperties, &NiagaraSpriteRendererProperties]
  }]
 userVariables [
  User.BurstScale {type=NiagaraFloat, defaultValue=1, description=""}
  User.CylinderHeight {type=NiagaraFloat, defaultValue=1, description=""}
  User.CylinderRadius {type=NiagaraFloat, defaultValue=0, description=""}
  User.Gravity {type=Vector3f, description="", defaultValue {struct=Vector3f, value=<0 0 -1500>}}
  User.InheritParentVelocity {type=Vector3f, description="", defaultValue {struct=Vector3f, value=<0.2 0.2 0.2>}}
  User.Shell Type {type=NiagaraInt32}
  User.ShellColor {type=LinearColor, description="", defaultValue {struct=LinearColor, value=<rgba 0 0 0 1>}}
  User.ShellsRotationRate {type=Vector3f, description="", defaultValue {struct=Vector3f, value=<1 1 1>}}
  User.Size {type=Vector3f, description="", defaultValue {struct=Vector3f, value=<1 1 1>}}
  User.SpawnProbability {type=NiagaraFloat, defaultValue=1, description=""}
  User.SpawnRateScale {type=NiagaraFloat, defaultValue=1, description=""}
  User.Velocity {type=Vector3f, description="", defaultValue {struct=Vector3f, value=<0 0 0>}}
 ]
}
