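# Physics asset for the Chrono pan rifle mesh, giving the weapon a simple physical collision shape.
# It holds a single body on the handle bone, built from one capsule tilted to lie roughly horizontally along the weapon.
# No constraints are defined, so the rifle simulates and collides as one rigid piece.
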PhysicsAsset /ChronoWeaponGameplay/Items/PanRifle/Meshes/Chrono_Pan_Rifle_PhysicsAsset
bodies {handle [{shapeType=Capsule, length=49.3553, radius=14.0734, center=<24.2778 -0.0351 6.667>, rotation=<rot 0 -90.1959 -89.9476>}]}
constraints []
