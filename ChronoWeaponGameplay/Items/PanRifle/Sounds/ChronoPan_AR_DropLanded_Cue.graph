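# Sound cue for the landing sound of the dropped Chrono Pan assault rifle.
# A Random node picks one of three equally weighted DropLanded wave variants and feeds it through a Modulator at the root.
# It plays in the weapon foley sound class with the shared Fort weapon drop attenuation.
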SoundCue /ChronoWeaponGameplay/Items/PanRifle/Sounds/ChronoPan_AR_DropLanded_Cue
root=Modulator_0
nodes [
 {id=Modulator_0, children [Random_0]}
 {id=Random_0, children [WavePlayer_0,WavePlayer_1,WavePlayer_2], properties {weights [1,1,1]}}
 {id=WavePlayer_0, properties {soundWaveAssetPtr=&/ChronoPan_AR_DropLanded_01}}
 {id=WavePlayer_1, properties {soundWaveAssetPtr=&/ChronoPan_AR_DropLanded_02}}
 {id=WavePlayer_2, properties {soundWaveAssetPtr=&/ChronoPan_AR_DropLanded_03}}
]
properties {
 attenuationSettings=&Fort_Weapon_Drop_Att
 maxDistance=3000
 soundClassObject=&Fort_WeaponFoley
 volumeMultiplier=1
}
