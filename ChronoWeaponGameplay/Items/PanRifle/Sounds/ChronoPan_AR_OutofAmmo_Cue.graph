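# Sound cue that plays the out-of-ammo sound for the ChronoPan assault rifle, the PanRifle item.
# A single wave player feeds the out-of-ammo wave through a modulator at the root.
# The cue uses the weapon foley sound class and the shared weapon reload attenuation, with slightly reduced volume.
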SoundCue /ChronoWeaponGameplay/Items/PanRifle/Sounds/ChronoPan_AR_OutofAmmo_Cue
root=Modulator_0
nodes [
 {id=Modulator_0, children [WavePlayer_0]}
 {id=WavePlayer_0, properties {soundWaveAssetPtr=&/ChronoPan_AR_OutofAmmo_01}}
]
properties {
 attenuationSettings=&Fort_Weapon_Reload_Att
 maxDistance=2250
 soundClassObject=&Fort_WeaponFoley
 volumeMultiplier=0.85
}
